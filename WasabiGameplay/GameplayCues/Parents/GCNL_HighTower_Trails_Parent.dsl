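; Parent looping gameplay cue that gives a player pawn a colored aura, movement-sensitive trails and loop/stop audio.
; On loop start it finds the aura by template, tints it, spawns mesh-attached trails and starts a repeating timer.
; The timer disables trails during skydiving, falling, jumping or high speed, and re-enables them otherwise.
; Removal swaps in the stop sound; ceasing relevance clears the timer and destroys the trails.
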(event Custom|SetColors
  (bind _p_aura_ref (Variables|Default|GetP_Aura_Ref))
  ;[1 1 1] Set the colors of the emitter and the alpha
  (Effects|Components|ParticleSystem|SetColorParameter _p_aura_ref "Basis Color" (Variables|Default|GetFX_Color))
  (Effects|Components|ParticleSystem|SetFloatParameter _p_aura_ref "Basis alpha" (Variables|Default|GetFX_Alpha)))

(event EventOnLoopingStart(EitherVFX-preferthisversion)
  (params (MyTarget :type Actor%20Object%20Reference) (Parameters :type Gameplay%20Cue%20Parameters%20Structure%20%28by%20ref%29) (ParticleComponents :type Array%20of%20FXSystem%20Component%20Object%20References) (AudioComponents :type Array%20of%20Audio%20Component%20Object%20References))
  ;[1 1 1] Save Out refs to the FX, set up the timer and call the color set, for the trails, spawn it dynamically so it can be toggled.
  (bind _asplayer_pawn_athena (Utilities|Casting|CastToPlayerPawn_Athena MyTarget))
  (Variables|Default|SetAthenaPlayerPawn _asplayer_pawn_athena)
  (Utilities|FlowControl|Sequence
    (:then_0
      (for _array_element ParticleComponents
        (bind _ascascade_particle_system_component (Utilities|Casting|CastToParticleSystemComponent _array_element))
        (if (Utilities|Equal(Object) (Variables|/Script/CascadeRuntime.ParticleSystemComponent|GetTemplate _ascascade_particle_system_component) (Variables|Default|GetEmitterTemplate_Aura))
          (Variables|Default|SetP_Aura_Ref _ascascade_particle_system_component)))
      (bind _returnvalue (Utilities|Time|SetTimerbyEvent (EventDispatchers|CreateEvent :!event Custom|CheckTrails) (Variables|Default|GetTrailsCheckInterval) true))
      (Variables|Default|SetTimer_TrailsCheck _returnvalue)
      (bind _returnvalue_2 (Effects|SpawnEmitterAttached :EmitterTemplate (Variables|Default|GetEmitterTemplate_Trails) :AttachToComponent (Variables|/Script/Engine.Character|GetMesh (Variables|Default|GetAthenaPlayerPawn)) :AttachPointName "None" :Scale "1.000000,1.000000,1.000000" :LocationType "SnapToTarget" :bAutoDestroy false :PoolingMethod "ManualRelease" :bAutoActivate true))
      (Variables|Default|SetP_Trails_Ref _returnvalue_2)
      (CallFunction|SetColors))
    (:then_1
      (bind _result (ArrayMacros|SafeArrayGet AudioComponents)
        (:%20%20
          (CallFunction|AudioComponentSetup _result true))))))

(event Custom|CheckTrails
  (bind _p_trails_ref (Variables|Default|GetP_Trails_Ref))
  (bind _returnvalue (Components|Activation|IsActive _p_trails_ref))
  ;[1 1 1] Check if the player is jumping, moving too fast, or skydiving.
  (if (Math|Boolean|ORBoolean (Parachute|IsSkydiving (Variables|Default|GetAthenaPlayerPawn)) (> (Variables|/Script/FortniteGame.FortPlayerPawnAthena|GetCurrentPawnSquaredSpeed (Variables|Default|GetAthenaPlayerPawn)) 8.1e+05) (AI|Components|NavMovement|IsFalling (Variables|/Script/Engine.Character|GetCharacterMovement (Variables|Default|GetAthenaPlayerPawn))) (Variables|/Script/Engine.Character|GetbWasJumping (Variables|Default|GetAthenaPlayerPawn)))
    (if _returnvalue
      (Components|Activation|Deactivate _p_trails_ref))
    (else
      (if (not _returnvalue)
        (Components|Activation|Activate _p_trails_ref true)))))

(event EventOnRemoval(EitherVFX-preferthisversion)
  (params (MyTarget :type Actor%20Object%20Reference) (Parameters :type Gameplay%20Cue%20Parameters%20Structure%20%28by%20ref%29) (ParticleComponents :type Array%20of%20FXSystem%20Component%20Object%20References) (AudioComponents :type Array%20of%20Audio%20Component%20Object%20References) (BurstCameraShakeInstance :type Camera%20Shake%20Base%20Object%20Reference) (BurstDecalInstance :type Decal%20Actor%20Object%20Reference))
  ;[1 1 1] Removal
  (bind _result (ArrayMacros|SafeArrayGet AudioComponents)
    (:%20%20
      (CallFunction|AudioComponentSetup _result))))

(fn AudioComponentSetup
  (params (AudioComponent :type Audio%20Component%20Object%20Reference) (Starting :type Boolean))
  (bind _output_get (LocalVariables|SetLocalAudioComponent AudioComponent))
  (Audio|Components|Audio|SetSound _output_get (select Starting (Variables|Audio|GetSoundOnLoop) (Variables|Audio|GetSoundOnStop)))
  (bind _output_get_2 (LocalVariables|SetLocalFPC (Fortnite|GetFirstLocalFortPlayerController (Variables|Default|GetAthenaPlayerPawn) true true))))

(fn OnCeaseRelevant
  (params (MyTarget :type Actor%20Object%20Reference) (Parameters :type Gameplay%20Cue%20Parameters%20Structure%20%28by%20ref%29))
  ;[1 1 1] Clear the Timer on Remove, Just in case.
  (bind _returnvalue (|CallParentFunction MyTarget Parameters))
  (Utilities|Time|ClearandInvalidateTimerbyHandle 0 (Variables|Default|GetTimer_TrailsCheck))
  (Components|DestroyComponent (Variables|Default|GetP_Trails_Ref) 0)
  (return _returnvalue))

(fn ConstructionScript)
